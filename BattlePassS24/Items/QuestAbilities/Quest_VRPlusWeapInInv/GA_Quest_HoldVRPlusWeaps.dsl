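; Battle pass quest ability that detects when the player holds enough Epic-or-better (VR+) weapons in inventory.
; It checks on activation and listens for pickup and valid NPC service events, delaying briefly so new items land.
; It gathers items by required rarity, filters them by weapon tags, then at the target count sends the CarryingVRPlusWeap quest stat and ends.
; ForLoopTest is an uncalled debug variant with screen prints.
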(event Ability|EventFailedToActivatePassiveAbility
  (params (ActorInfo :type Gameplay%20Ability%20Actor%20Info%20Structure))
  ;[1 1 1] Check if player has a weapon of any of the specified rarity types in inventory ~LS 3/20/23
  (|CallParentFunction ActorInfo))

(event Ability|EventActivateAbility
  ;[1 1 1] Check if player has a weapon of any of the specified rarity types in inventory ~LS 3/20/23
  (|CallParentFunction)
  (Ability|Tasks|WaitGameplayEvent "(TagName=\"Event.CollectedPickup\")"
    (:then
      (bind (_asynctaskproxy _payload) (Ability|Tasks|WaitGameplayEvent "(TagName=\"Event.NPCService\")" 0 false false)
        (:then
          ;;[1 1 1] Initial Inventory check when we get the quest.
          (Utilities|FlowControl|DoOnce false
            (:Completed
              _)))
        (:EventReceived
          (bind (_payload_eventtag _payload_instigator _payload_target _payload_optionalobject _payload_optionalobject2 _payload_contexthandle _payload_instigatortags) (Utilities|Struct|BreakGameplayEventData _payload))
          (if (GameplayTags|HasAnyTags _payload_instigatortags (Variables|Default|GetValidNPCServices) true)
            _)))
      (Utilities|Array|Clear (Variables|Default|GetPlayerWeaponsOfReqRarities))
      (Utilities|Array|Clear (Variables|Default|GetPlayerItemsOfReqRarities))
      ;;[1 1 1] Delay a moment so the latest pickup item gets into the players inventory
      (Utilities|FlowControl|Delay 1.0)
      (Variables|Default|SetOwningPlayerInventoryInterface (Fortnite|GetFortPlayerControllerfromActor (Variables|GAVars|Data|GetOwningPlayerPawn)))
      (CallFunction|CheckWeaponsRarity)
      (CallFunction|HasReqWeaponsRarity))
    (:EventReceived
      _)))

(event Custom|Check%20Weapons%20Rarity
  ;[1 1 1] Check if player has a weapon of any of the specified rarity types in inventory ~LS 3/20/23
  ;;[1 1 1] Get all items of the specified rarities and combine to one array
  (for _array_element (Variables|Default|GetRequiredRarityTypes)
    (bind _items (InventoryManagement|GetItemsfromInventoryOwnerforRarity (Variables|Default|GetOwningPlayerInventoryInterface) _array_element))
    (if (Utilities|Array|IsNotEmpty _items)
      (Utilities|Array|AppendArray (Variables|Default|GetPlayerItemsOfReqRarities) _items))))

(event Custom|Has%20Req%20Weapons%20Rarity
  ;[1 1 1] Check if player has a weapon of any of the specified rarity types in inventory ~LS 3/20/23
  ;;[1 1 1] Using the array of req rarities, check if they are weapons that count for this quest. If they do, add them to a new array of Epic+ weapons
  (bind _array_element (Utilities|Array|ForEachLoopwithBreak (Variables|Default|GetPlayerItemsOfReqRarities))
    (:LoopBody
      (if (and (GameplayTags|HasAnyTags (CallFunction|GetOwnedGameplayTags _array_element) (Variables|Default|GetT_RequiredWeaponTags)) (not (GameplayTags|HasAnyTags (CallFunction|GetOwnedGameplayTags _array_element) (Variables|Default|GetT_ExcludedWeaponTags) true)))
        (Utilities|Array|AddUnique (Variables|Default|GetPlayerWeaponsOfReqRarities) _array_element)
        ;;[1 1 1] Do we have 3 Epic+ weapons in this new array? If yes, set bool to True and break the loop.
        (if (>= (Utilities|Array|Length (Variables|Default|GetPlayerWeaponsOfReqRarities)) (Variables|Default|GetAmountToHold))
          (Variables|Default|SetThreeVRPlusWeapsFound true))
        (Variables|Default|SetThreeVRPlusWeapsFound false)))
    (:Completed
      _))
  ;;[1 1 1] If bool is true at the end of the loop, give the player the CarryingVRPlusWeap Tag and end the ability
  (if (Variables|Default|GetThreeVRPlusWeapsFound)
    (Quest|SendComplexCustomStatEvent (Variables|GAVars|Data|GetOwningPlayerQuestManager) 0 "(GameplayTags=((TagName=\"Athena.Quests.CarryingVRPlusWeap\")))" "(GameplayTags=)" 1)
    (Ability|EndAbility)))

(event Custom|ForLoopTest
  (bind _playeritemsofreqrarities (Variables|Default|GetPlayerItemsOfReqRarities))
  (bind _index (Utilities|FlowControl|ForLoopwithBreak 0 (- (Utilities|Array|Length _playeritemsofreqrarities) 1))
    (:LoopBody
      (bind _output_get (Variables|Default|SetArrayItem (Utilities|Array|Get(acopy) _playeritemsofreqrarities _index)))
      (if (and (GameplayTags|HasAnyTags (GameplayTags|GetOwnedGameplayTags _output_get :!class /Script/GameplayTags.GameplayTagAssetInterface) (Variables|Default|GetT_RequiredWeaponTags)) (not (GameplayTags|HasAnyTags (GameplayTags|GetOwnedGameplayTags _output_get :!class /Script/GameplayTags.GameplayTagAssetInterface) (Variables|Default|GetT_ExcludedWeaponTags) true)))
        (Development|PrintString (Utilities|String|Append (Utilities|String|ToString(Text) (Item|GetDisplayName _output_get)) " is a weapon") true true "(R=0.000000,G=0.671875,B=0.031178,A=1.000000)" 6.0)
        (Utilities|Array|AddUnique (Variables|Default|GetPlayerWeaponsOfReqRarities) _output_get)
        (if (>= (Utilities|Array|Length (Variables|Default|GetPlayerWeaponsOfReqRarities)) (Variables|Default|GetAmountToHold))
          (Variables|Default|SetThreeVRPlusWeapsFound true))
        (Variables|Default|SetThreeVRPlusWeapsFound false))
      (Development|PrintString (Utilities|String|Append (Utilities|String|ToString(Text) (Item|GetDisplayName _output_get)) " is NOT a weapon") true true "(R=1.000000,G=0.000000,B=0.226433,A=1.000000)" 6.0))
    (:Completed
      _))
  ;[1 1 1] If bool is true at the end of the loop, give the player the CarryingVRPlusWeap Tag and end the ability
  (if (Variables|Default|GetThreeVRPlusWeapsFound)
    (Quest|SendComplexCustomStatEvent (Variables|GAVars|Data|GetOwningPlayerQuestManager) 0 "(GameplayTags=((TagName=\"Athena.Quests.CarryingVRPlusWeap\")))" "(GameplayTags=)" 1)
    (Ability|EndAbility)))

(fn GetOwnedGameplayTags
  (params (self2 :type Gameplay%20Tag%20Asset%20Interface%20Interface))
  (return (GameplayTags|GetOwnedGameplayTags self2 :!class /Script/GameplayTags.GameplayTagAssetInterface)))
